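# Biped Control Rig module that draws colored debug lines showing stretch and squash along the spine, arms and legs.
# Construction resolves element connectors, derives right limbs by renaming _l to _r, and adds a visibility bool channel when a control is connected.
# Each update, a shared Stretch Feedback function, gated by that channel, compares consecutive elements' current and initial distances.
# That ratio blends line color between squash, rest and stretch colors.
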ControlRigBlueprint /EpicControlRig/Modules/Biped/StretchFeedback/CRM_Epic_BipedStretchFeedback_v01
graphs [
 ConstructionGraph {nodes [
   RigUnit_PrepareForExecution = &RigUnit_PrepareForExecution()
   RigVMFunction_Sequence = RigVMAggregateNode(ExecuteContext=@RigUnit_PrepareForExecution.ExecutePin)
   n1 = /Script/ControlRig.RigUnit_ResolveArrayConnector(Connector=(Type=Connector,Name="Arm Elements"), SkipSocket=False) {is=struct RigVMUnitNode}
   RigVMVariableNode(Value=@n1.Result, ExecuteContext=@RigVMFunction_Sequence.B) {variable=Arm L Elements}
   Clone = RigVMDispatchNode(Array=@n1.Result) {template="DISPATCH_RigVMDispatch_ArrayClone(in Array,out Clone)"}
   > RigVMVariableNode(Value=/Script/ControlRig.RigUnit_CollectionReplaceItemsArray(New=_r, Old=_l, RemoveInvalidItems=False, bAllowDuplicates=False, Items=@Clone.Clone) {is=struct RigVMUnitNode}.Result) {variable=Arm R Elements}
   RigVMVariableNode(
    Value=&RigUnit_ResolveArrayConnector(Connector=(Type=Connector,Name="Spine Elements"), SkipSocket=False).Result
    ExecuteContext=@RigVMFunction_Sequence.A
   ) {variable=Spine Elements}
   n2 = &RigUnit_ResolveArrayConnector(Connector=(Type=Connector,Name="Leg Elements"), SkipSocket=False)
   RigVMVariableNode(Value=@n2.Result, ExecuteContext=@RigVMFunction_Sequence.C) {variable=Leg L Elements}
   > RigVMVariableNode(Value=&RigUnit_CollectionReplaceItemsArray(
     New=_r
     Old=_l
     RemoveInvalidItems=False
     bAllowDuplicates=False
     Items=RigVMDispatchNode(Array=@n2.Result) {template="DISPATCH_RigVMDispatch_ArrayClone(in Array,out Clone)"}.Clone
    ).Result) {variable=Leg R Elements}
   n3 = /Script/ControlRig.RigUnit_ResolveConnector(Connector=(Type=Connector,Name="Vis Channel Control"), SkipSocket=False) {is=struct RigVMUnitNode}
   RigVMFunction_ControlFlowBranch = &RigVMFunction_ControlFlowBranch(Condition=@n3.bIsConnected, ExecuteContext=@RigVMFunction_Sequence.D)
   SpawnAnimationChannel = /Script/ControlRig.RigUnit_HierarchyAddAnimationChannelBool(
    InitialValue=False
    LimitsEnabled=()
    MaximumValue=True
    MinimumValue=False
    Name=Stretch FeedBack Vis
    Parent=@n3.Result
    ExecutePin=@RigVMFunction_ControlFlowBranch.True
   ) {
    is=struct RigVMUnitNode
    template="SpawnAnimationChannel::Execute(in InitialValue,in MinimumValue,in MaximumValue,in LimitsEnabled,in Parent,in Name,out Item)"
   }
   > RigVMVariableNode(Value=@SpawnAnimationChannel.Item) {variable=Vis Channel}
  ]}
 New Function_ContainedGraph {nodes [
   n1 = RigVMFunctionEntryNode()
   RigVMFunction_Sequence = &RigVMFunction_Sequence(ExecuteContext=@n1.ExecuteContext)
   n2 = &RigVMFunction_ControlFlowBranch(ExecuteContext=@RigVMFunction_Sequence.A, Condition=RigVMVariableNode() {variable=Enabled}.Value)
   For_Each = RigVMDispatchNode(ExecuteContext=@n2.True, Array=RigVMVariableNode() {variable=Elements}.Value) {template="DISPATCH_RigVMDispatch_ArrayIterator(in Array,out Element,out Index,out Count,out Ratio)"}
   Not_Equals = RigVMDispatchNode(B=0, A=@For_Each.Index) {template="DISPATCH_RigVMDispatch_CoreNotEquals(in A,in B,out Result)"}
   RigVMFunction_ControlFlowBranch = &RigVMFunction_ControlFlowBranch(ExecuteContext=@For_Each.ExecuteContext, Condition=@Not_Equals.Result)
   n3 = &RigUnit_GetTransform(bInitial=False, Item=@For_Each.Element)
   Subtract = /Script/RigVM.RigVMFunction_MathIntSub(B=1, A=@For_Each.Index) {is=struct RigVMUnitNode, template="Subtract::Execute(in A,in B,out Result)"}
   n4 = RigVMDispatchNode(Index=@Subtract.Result, Array=RigVMVariableNode() {variable=Elements}.Value) {template="DISPATCH_RigVMDispatch_ArrayGetAtIndex(in Array,in Index,out Element)"}
   n5 = &RigUnit_GetTransform(bInitial=False, Item=@n4.Element)
   RigVMFunction_MathVectorDistance = &RigVMFunction_MathVectorDistance(A=@n3.Transform.Translation, B=@n5.Transform.Translation)
   Divide = &RigVMFunction_MathDoubleDiv(
    A=@RigVMFunction_MathVectorDistance.Result
    B=&RigVMFunction_MathVectorDistance(
     A=&RigUnit_GetTransform(bInitial=true, Item=@For_Each.Element).Transform.Translation
     B=&RigUnit_GetTransform(bInitial=true, Item=@n4.Element).Transform.Translation
    ).Result
   )
   Greater = /Script/RigVM.RigVMFunction_MathDoubleGreater(B=1, A=@Divide.Result) {is=struct RigVMUnitNode, template="Greater::Execute(in A,in B,out Result)"}
   Remap = &RigVMFunction_MathDoubleRemap(
    SourceMinimum=1
    TargetMaximum=1
    TargetMinimum=0
    bClamp=false
    Value=@Divide.Result
    SourceMaximum=RigVMVariableNode() {variable=Max Stretch Factor}.Value
   )
   Interpolate = /Script/RigVM.RigVMFunction_MathColorLerp(T=@Remap.Result, A=RigVMVariableNode() {variable=Squash Stretch Start Color}.Value, B=RigVMVariableNode() {variable=Stretch Color}.Value) {is=struct RigVMUnitNode, template="Interpolate::Execute(in A,in B,in T,out Result)"}
   n6 = &RigVMFunction_MathColorLerp(A=RigVMVariableNode() {variable=Squash Color}.Value, B=RigVMVariableNode() {variable=Squash Stretch Start Color}.Value, T=@Divide.Result) {template="Interpolate::Execute(in A,in B,in T,out Result)"}
   IsNearlyEqual = /Script/RigVM.RigVMFunction_MathDoubleIsNearlyEqual(B=1, Tolerance=0.01, A=@Divide.Result) {is=struct RigVMUnitNode, template="IsNearlyEqual::Execute(in A,in B,in Tolerance,out Result)"}
   RigVMFunction_DebugLineNoSpace = /Script/RigVM.RigVMFunction_DebugLineNoSpace(
    DebugDrawSettings=(DepthPriority=SDPG_Foreground,Lifetime=-1)
    WorldOffset=(Rotation=(X=0,Y=0,Z=0,W=1),Translation=(X=0,Y=0,Z=0),Scale3D=(X=1,Y=1,Z=1))
    bEnabled=true
    Color=RigVMDispatchNode(
     False=RigVMDispatchNode(Condition=@Greater.Result, True=@Interpolate.Result, False=@n6.Result) {template="DISPATCH_RigVMDispatch_If(in Condition,in True,in False,out Result)"}.Result
     Condition=@IsNearlyEqual.Result
     True=RigVMVariableNode() {variable=Rest Color}.Value
    ) {template="DISPATCH_RigVMDispatch_If(in Condition,in True,in False,out Result)"}.Result
    ExecutePin=@RigVMFunction_ControlFlowBranch.True
    Thickness=RigVMVariableNode() {variable=Line Thickness}.Value
    B=@n5.Transform.Translation
    A=@n3.Transform.Translation
   ) {is=struct RigVMUnitNode}
   RigVMFunctionReturnNode(ExecuteContext=@RigVMFunction_Sequence.B)
  ]}
 RigVMFunctionLibrary {nodes [Stretch Feedback = RigVMCollapseNode(
    Elements=()
    ExecuteContext=()
    Line Thickness=2
    Max Stretch Factor=1
    Rest Color=(R=0,G=0,B=0,A=1)
    Squash Color=(R=0,G=0,B=0,A=1)
    Squash Stretch Start Color=(R=0,G=0,B=0,A=0)
    Stretch Color=(R=1,G=1,B=1,A=1)
   )]}
 RigVMFunction_Sequence_1_ContainedGraph {nodes [
   n1 = RigVMFunctionEntryNode()
   RigVMFunction_Sequence = &RigVMFunction_Sequence(ExecuteContext=@n1.ExecuteContext)
   n2 = &RigVMFunction_Sequence(ExecuteContext=@RigVMFunction_Sequence.B)
   RigVMFunctionReturnNode(A=@RigVMFunction_Sequence.A, B=@n2.A, C=@n2.B)
  ]}
 RigVMFunction_Sequence_1_ContainedGraph {nodes [
   n1 = RigVMFunctionEntryNode()
   RigVMFunction_Sequence = &RigVMFunction_Sequence(ExecuteContext=@n1.ExecuteContext)
   n2 = &RigVMFunction_Sequence(ExecuteContext=@RigVMFunction_Sequence.B)
   n3 = &RigVMFunction_Sequence(ExecuteContext=@n2.B)
   RigVMFunctionReturnNode(A=@RigVMFunction_Sequence.A, B=@n2.A, C=@n3.A, D=@n3.B)
  ]}
 RigVMModel {nodes [
   RigUnit_BeginExecution = &RigUnit_BeginExecution()
   RigVMFunction_Sequence = RigVMAggregateNode(ExecuteContext=@RigUnit_BeginExecution.ExecutePin)
   GetAnimationChannelFromItem = /Script/ControlRig.RigUnit_GetBoolAnimationChannelFromItem(bInitial=False, Item=RigVMVariableNode() {variable=Vis Channel}.Value) {is=struct RigVMUnitNode, template="GetAnimationChannelFromItem::Execute(out Value,in Item,in bInitial)"}
   Stretch Feedback_1 = RigVMFunctionReferenceNode(
    Elements=RigVMVariableNode() {variable=Arm L Elements}.Value
    Squash Stretch Start Color=RigVMVariableNode() {variable=Squash Stretch Start Color}.Value
    Rest Color=RigVMVariableNode() {variable=Rest Color}.Value
    Stretch Color=RigVMVariableNode() {variable=Stretch Color}.Value
    Squash Color=RigVMVariableNode() {variable=Squash Color}.Value
    Line Thickness=RigVMVariableNode() {variable=Line Thickness}.Value
    Max Stretch Factor=RigVMVariableNode() {variable=Max Stretch Factor}.Value
    ExecuteContext=@RigVMFunction_Sequence.B
    Enabled=@GetAnimationChannelFromItem.Value
   )
   > Stretch Feedback_2 = RigVMFunctionReferenceNode(
    Line Thickness=RigVMVariableNode() {variable=Line Thickness}.Value
    Max Stretch Factor=RigVMVariableNode() {variable=Max Stretch Factor}.Value
    Elements=RigVMVariableNode() {variable=Arm R Elements}.Value
    Rest Color=RigVMVariableNode() {variable=Rest Color}.Value
    Stretch Color=RigVMVariableNode() {variable=Stretch Color}.Value
    Squash Color=RigVMVariableNode() {variable=Squash Color}.Value
    Squash Stretch Start Color=RigVMVariableNode() {variable=Squash Stretch Start Color}.Value
    Enabled=@GetAnimationChannelFromItem.Value
   )
   Stretch Feedback_3 = RigVMFunctionReferenceNode(
    Max Stretch Factor=RigVMVariableNode() {variable=Max Stretch Factor}.Value
    Squash Stretch Start Color=RigVMVariableNode() {variable=Squash Stretch Start Color}.Value
    Elements=RigVMVariableNode() {variable=Leg L Elements}.Value
    Rest Color=RigVMVariableNode() {variable=Rest Color}.Value
    Squash Color=RigVMVariableNode() {variable=Squash Color}.Value
    Stretch Color=RigVMVariableNode() {variable=Stretch Color}.Value
    Line Thickness=RigVMVariableNode() {variable=Line Thickness}.Value
    ExecuteContext=@RigVMFunction_Sequence.C
    Enabled=@GetAnimationChannelFromItem.Value
   )
   > Stretch Feedback_4 = RigVMFunctionReferenceNode(
    Line Thickness=RigVMVariableNode() {variable=Line Thickness}.Value
    Max Stretch Factor=RigVMVariableNode() {variable=Max Stretch Factor}.Value
    Rest Color=RigVMVariableNode() {variable=Rest Color}.Value
    Stretch Color=RigVMVariableNode() {variable=Stretch Color}.Value
    Squash Color=RigVMVariableNode() {variable=Squash Color}.Value
    Squash Stretch Start Color=RigVMVariableNode() {variable=Squash Stretch Start Color}.Value
    Elements=RigVMVariableNode() {variable=Leg R Elements}.Value
    Enabled=@GetAnimationChannelFromItem.Value
   )
   Stretch Feedback = RigVMFunctionReferenceNode(
    Elements=RigVMVariableNode() {variable=Spine Elements}.Value
    Rest Color=RigVMVariableNode() {variable=Rest Color}.Value
    Stretch Color=RigVMVariableNode() {variable=Stretch Color}.Value
    Squash Color=RigVMVariableNode() {variable=Squash Color}.Value
    Squash Stretch Start Color=RigVMVariableNode() {variable=Squash Stretch Start Color}.Value
    Max Stretch Factor=RigVMVariableNode() {variable=Max Stretch Factor}.Value
    Line Thickness=RigVMVariableNode() {variable=Line Thickness}.Value
    ExecuteContext=@RigVMFunction_Sequence.A
    Enabled=@GetAnimationChannelFromItem.Value
   )
  ]}
]
hierarchy [
 Root {type=Connector, loc=<0 0 0>, rot=<0 0 0 1>}
 Root {type=Socket, loc=<0 0 0>, rot=<0 0 0 1>}
 Spine Elements {type=Connector, loc=<0 0 0>, rot=<0 0 0 1>}
 Arm Elements {type=Connector, loc=<0 0 0>, rot=<0 0 0 1>}
 Leg Elements {type=Connector, loc=<0 0 0>, rot=<0 0 0 1>}
 Vis Channel Control {type=Connector, loc=<0 0 0>, rot=<0 0 0 1>}
]
variables [
 Arm L Elements {type="TArray<FRigElementKey>"}
 Arm R Elements {type="TArray<FRigElementKey>"}
 Leg L Elements {type="TArray<FRigElementKey>"}
 Leg R Elements {type="TArray<FRigElementKey>"}
 Line Thickness {type=double, default=1}
 Max Stretch Factor {type=double, default=5}
 Rest Color {type=FLinearColor, default=(R=0,G=1,B=0,A=1)}
 Spine Elements {type="TArray<FRigElementKey>"}
 Squash Color {type=FLinearColor, default=(R=0,G=0,B=1,A=1)}
 Squash Stretch Start Color {type=FLinearColor, default=(R=1,G=1,B=1,A=1)}
 Stretch Color {type=FLinearColor, default=(R=1,G=0,B=0,A=1)}
 Vis Channel {type=FRigElementKey, default=(Type=None,Name="")}
]
